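# Editor-only PistolCanary Level Sequence, marked for delete by name, that stages a third-person-style gameplay shot with Fortnite player mannequins.
# Mannequins follow densely keyed, recorded-looking paths with skeletal animation, alongside a separately animated transform reference mesh.
# A 14mm cine camera holds a fixed attached offset, while an attached vertical tilt pivot carries keyed rotation.
# An assault rifle is attached with toggled visibility, and a dead-tree prop has a visibility override.
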LevelSequence /PistolCanary/Sequence/PistolCanary_EditorOnly/PistolCanary_EditorOnly1000/MarkedForDelete_PistolCanary_EditorOnly1000
bindings [
 BP_RustBucket_DeadTree_SaltLake_F {
  id=#1
  possessed_class=BP_Hermes_MaskMaker_DeadTree_E_C
  tracks [{
    type=MovieSceneVisibilityTrack
    sections [{type=MovieSceneVisibilitySection, channels [None {type=MovieSceneScriptingBoolChannel, default=false}]}]
   }]
 }
 _camera_vertical_tilt_pivot {
  id=#2
  possessed_class=FortStaticMeshActor
  tracks [
   {type=MovieSceneSpawnTrack, sections [{type=MovieSceneSpawnSection, channels [None {type=MovieSceneScriptingBoolChannel, default=true}]}]}
   {
    type=MovieScene3DTransformTrack
    sections [{
      type=MovieScene3DTransformSection
      channels [
       Location.X {type=MovieSceneScriptingDoubleChannel, default=0, const [0,15]}
       Location.Y {type=MovieSceneScriptingDoubleChannel, default=0, const [0,15]}
       Location.Z {type=MovieSceneScriptingDoubleChannel, default=0, const [0,15]}
       Rotation.X {
        type=MovieSceneScriptingDoubleChannel
        default=0
        thinned_from=18
        keys [[-6293,0],[5397,0],[5434,-4.831],[5759,-13.4885],[5820,-13.9201],[5971,-13.9201],[6055,-20.1557],[6165,-23.4323],[6225,-23.8509],[7807,-23.8509],[7832,-22.021],[7862,-21.5048],[9010,-21.5048],[9038,-34.4363],[9063,-36.6547]]
       }
       Rotation.Y {type=MovieSceneScriptingDoubleChannel, default=0, thinned_from=15, keys [[-6293,0],[9063,0]]}
       Rotation.Z {
        type=MovieSceneScriptingDoubleChannel
        default=37
        thinned_from=19
        keys [[-6293,0],[4791,0],[4844,20.0439],[4895,18.3805],[4983,-15.6548],[5068,-15.6548],[5134,-0.8798],[5161,0],[9063,0]]
       }
       Scale.X {type=MovieSceneScriptingDoubleChannel, default=1, const [1,15]}
       Scale.Y {type=MovieSceneScriptingDoubleChannel, default=1, const [1,15]}
       Scale.Z {type=MovieSceneScriptingDoubleChannel, default=1, const [1,15]}
      ]
     }]
   }
   {type=MovieScene3DAttachTrack, sections [{type=MovieScene3DAttachSection}]}
  ]
 }
 CameraComponent {
  id=#3
  parent_id=#4
  possessed_class=CineCameraComponent
  tracks [
   {
    type=MovieSceneFloatTrack
    display_name=Current Focal Length
    sections [{type=MovieSceneFloatSection, channels [None {type=MovieSceneScriptingFloatChannel, default=14, const [14,2]}]}]
   }
   {
    type=MovieSceneFloatTrack
    display_name="Manual Focus Distance (Focus Settings)"
    sections [{type=MovieSceneFloatSection, channels [None {type=MovieSceneScriptingFloatChannel, default=100000}]}]
   }
   {
    type=MovieSceneFloatTrack
    display_name=Current Aperture
    sections [{type=MovieSceneFloatSection, channels [None {type=MovieSceneScriptingFloatChannel, default=2.8}]}]
   }
  ]
 }
 FortPlayerMannequin {
  id=#5
  possessed_class=FortPlayerMannequin
  tracks [
   {type=MovieSceneSpawnTrack, sections [{type=MovieSceneSpawnSection, channels [None {type=MovieSceneScriptingBoolChannel, default=true}]}]}
   {type=MovieSceneSkeletalAnimationTrack, sections [{type=MovieSceneSkeletalAnimationSection, end_frame=10274, start_frame=3480}]}
   {
    type=MovieScene3DTransformTrack
    sections [{
      type=MovieScene3DTransformSection
      end_frame=10274
      start_frame=3480
      channels [
       Location.X {
        type=MovieSceneScriptingDoubleChannel
        default=16723.9288
        thinned_from=119
        keys [[3480,16356.1896],[3924.6975,16317.1521],[3963.2175,16099.9794],[4080.1575,15856.2799],[4139.3175,15852.2848],[4152.1,15885.0696],[4891.2975,15753.1506],[4958.3775,15504.4456],[4982.74,15565.8352],[5000.9175,15587.8824],[5026.3575,15547.6163],[5053.9575,15457.3575],[7831.5975,15420.3171],[7870,15180.2001],[7879.3575,15153.7158],[7884.1575,15110.4629],[7891.5975,15082.5144],[10274,15072.145]]
       }
       Location.Y {
        type=MovieSceneScriptingDoubleChannel
        default=80673.3294
        thinned_from=119
        keys [[3480,80168.6637],[5078.0775,78609.23],[10274,78672.4378]]
       }
       Location.Z {
        type=MovieSceneScriptingDoubleChannel
        default=3263.4275
        thinned_from=62
        keys [[3480,3075.7243],[3935.9175,3074.0066],[3943.3575,3089.1724],[4946.2,3087.5644],[4950.0975,3075.013],[5078.0775,3067.9359],[7840.96,3066.1702],[7855.24,3048.3656],[10274,3048.5591]]
       }
       Rotation.X {type=MovieSceneScriptingDoubleChannel, default=0, const [0,3]}
       Rotation.Y {type=MovieSceneScriptingDoubleChannel, default=0, const [0,3]}
       Rotation.Z {
        type=MovieSceneScriptingDoubleChannel
        default=-138.6945
        thinned_from=555
        keys [[3480,-220.9893],[3489.3375,-220.1398],[3506.6175,-220.1398],[3523.9,-220.9175],[3545.4375,-221.5396],[3556.42,-223.095],[3567.1575,-226.9836],[3570.82,-228.6945],[3585.52,-232.1164],[3629.3775,-237.2492],[3639.9375,-237.7158],[3643.48,-238.338],[3700.4175,-238.4935],[3751.3575,-235.5383],[3822.1575,-232.8941],[3853.5975,-230.4054],[3870.9975,-230.4054],[3881.4375,-229.6277],[3905.68,-229.4722],[3920.5575,-225.5837],[3935.9175,-224.3394],[3947.3775,-224.3394],[3955.4175,-223.4061],[3959.3175,-222.1618],[3971.3175,-215.4736],[3979.78,-212.6739],[3992.5575,-212.2072],[4005.04,-212.2072],[4013.62,-212.8294],[4109.1375,-212.8294],[4117.5975,-214.8514],[4126.5375,-215.1625],[4130.7975,-217.029],[4147.96,-220.7619],[4161.5775,-221.073],[4173.3375,-219.8287],[4177.1775,-218.8955],[4184.86,-216.4068],[4212.1575,-214.2293],[4230.46,-214.0737],[4241.4375,-212.8294],[4268.92,-211.5851],[4651.9575,-211.274],[4666.6575,-212.2072],[4784.86,-211.8962],[4799.5575,-211.1185],[4853.5575,-210.0297],[4860.9375,-208.9409],[4879.78,-208.0077],[4911.2175,-208.0077],[4915.0575,-207.0744],[4926.8775,-201.1639],[4930.8375,-199.9196],[4942.4775,-198.0531],[4962.82,-198.0531],[4966.7175,-197.4309],[5026.3575,-197.4309],[5032.7175,-196.6533],[5053.9575,-196.4977],[5060.4375,-197.742],[5072.0175,-197.742],[5090.0175,-196.6533],[5100.46,-195.2534],[5105.9175,-194.9423],[5119.1175,-192.7647],[5437.54,-193.5425],[5451.9375,-194.3201],[5479.8975,-194.7868],[5518.1175,-194.9423],[5525.1975,-195.5645],[5543.32,-196.1866],[5592.7575,-197.1199],[5738.32,-197.2754],[5753.1975,-197.8976],[5866.7775,-197.5865],[5870.74,-197.2754],[5928.5175,-196.0311],[5945.7975,-196.0311],[5966.9175,-195.2534],[6293.3775,-195.0978],[6307.2975,-194.4757],[6424.2975,-194.1646],[6675.6375,-194.1646],[6679.6,-193.2964],[6740.74,-193.4145],[6747.9375,-193.7688],[6813.6375,-192.706],[6849.3375,-191.9974],[6969.1575,-192.3517],[6972.6975,-192.5879],[7223.32,-190.6984],[7434.0375,-189.5175],[7451.7375,-190.5803],[7540.54,-189.6356],[7654.4175,-189.8786],[7662.0375,-189.1009],[7670.1975,-188.7898],[7712.32,-188.6342],[7721.0775,-189.6884],[7748.4975,-188.8618],[7763.7375,-190.1608],[7796.3775,-190.1608],[7809.4575,-190.9251],[7822.6575,-190.9251],[7831.5975,-190.1474],[7836.0975,-189.3697],[7879.3575,-188.592],[7884.1575,-188.1254],[7891.5975,-188.1254],[7902.1575,-187.6588],[7914.8775,-186.57],[7921.6575,-185.6367],[7927.8975,-185.3257],[7934.1375,-184.3924],[7940.1975,-184.0813],[7968.9375,-184.3924],[7996.42,-185.1701],[8611.5375,-185.9478],[8916.0975,-186.1033],[8923.6575,-187.1921],[8942.0775,-187.3477],[8949.8775,-186.881],[8971.5975,-188.4365],[9027.3375,-187.0192],[9123.8775,-186.3107],[9162.28,-186.3107],[9176.8575,-185.9564],[9191.7375,-185.0117],[9205.96,-185.1298],[9216.8175,-185.8383],[9232.0575,-188.082],[9239.5575,-188.3182],[9243.2775,-189.0268],[9265.4775,-189.9715],[9344.0175,-189.9715],[9355.0575,-188.9087],[9379.8975,-188.5544],[9390.76,-187.3735],[9408.76,-186.4288],[9758.32,-185.6511],[9765.3375,-184.4067],[9772.1175,-182.6958],[9779.1375,-182.2292],[9799.4175,-182.3847],[9802.8975,-182.6958],[10274,-182.6958]]
       }
       Scale.X {type=MovieSceneScriptingDoubleChannel, default=1, const [1,3]}
       Scale.Y {type=MovieSceneScriptingDoubleChannel, default=1, const [1,3]}
       Scale.Z {type=MovieSceneScriptingDoubleChannel, default=1, const [1,3]}
      ]
     }]
   }
  ]
 }
 Mann_Transform_ref {
  id=#6
  possessed_class=FortStaticMeshActor
  tracks [
   {type=MovieSceneSpawnTrack, sections [{type=MovieSceneSpawnSection, channels [None {type=MovieSceneScriptingBoolChannel, default=true}]}]}
   {
    type=MovieScene3DTransformTrack
    sections [{
      type=MovieScene3DTransformSection
      end_frame=20015
      start_frame=0
      channels [
       Location.X {
        type=MovieSceneScriptingDoubleChannel
        default=12518.7627
        thinned_from=721
        keys [[0,12518.7627],[56.16,12518.7627],[59.1,12461.8179],[65.28,12461.7249],[68.28,12575.7132],[81.3,12575.2068],[84.72,12591.5155],[219.66,12863.6757],[244.86,12953.8869],[274.08,13084.4146],[518.64,13050.6472],[1025.04,13050.6472],[1028.04,12575.708],[1039.5,12553.3362],[1042.8,12575.708],[1096.38,12579.1843],[1110.9,12528.6313],[1132.14,12461.0809],[1167.06,12412.2783],[1183.14,12348.9814],[1269,12516.4879],[1285.56,12659.1383],[1360.44,12983.8659],[1371.72,13047.5056],[1378.62,13063.0924],[1386.36,13101.5017],[6478.98,13127.4064],[6502.32,13230.6491],[6543.84,13528.7273],[6628.14,14015.7203],[6635.58,14036.8616],[6639,14080.1078],[6666.18,14235.1216],[6756.9,14285.3459],[6805.26,14358.0482],[7335.84,14299.0721],[7360.92,14182.9396],[7428.48,13764.0294],[7487.7,13910.1725],[8265.66,13910.1725],[8268.96,12575.708],[8346.48,12573.387],[8353.92,12596.1576],[8357.46,12573.1324],[8394.24,12680.9903],[8543.4,12674.9941],[8646.78,13059.5369],[10677.78,13096.3471],[10701.18,13234.0858],[10798.02,13791.8614],[10818.48,13918.5875],[10975.26,13853.7341],[11024.34,13657.4805],[11042.58,13680.6295],[11511.78,13663.2734],[11614.5,13649.495],[11630.82,13709.4822],[11653.14,13721.4827],[11704.38,13914.9671],[11796.3,14141.4266],[11832.72,14180.5917],[11842.74,14208.5301],[12417.3,14208.5301],[12420.84,12575.708],[12504.78,12575.539],[12512.46,12614.2782],[12550.14,12627.1718],[12569.28,12699.109],[12739.98,12715.7493],[12782.52,12853.1341],[12862.14,12868.37],[12877.86,12936.781],[12881.46,12987.1411],[12901.68,13064.3454],[12965.1,13019.2784],[13266.9,13019.2784],[13270.08,12575.708],[13351.74,12562.1224],[13359.78,12610.0038],[13367.88,12610.0038],[13371.66,12588.1637],[13414.98,12666.7092],[13494.12,12688.8235],[13593.18,13062.0434],[13604.64,13094.8048],[15735.66,13126.0941],[15744.3,13159.4657],[15790.2,13438.742],[15795.96,13458.0086],[15879.72,13957.503],[15885.84,13977.5713],[15889.2,14015.1572],[15895.5,14037.0736],[15899.04,14078.5902],[15952.98,14163.6001],[15972.18,14263.2656],[15975.48,14298.9949],[15984.42,14322.0424],[18418.98,14341.739],[18443.4,14430.8906],[18673.5,14406.3388],[18701.58,14296.4523],[18705.06,14261.1951],[19663.08,14255.0905]]
       }
       Location.Y {
        type=MovieSceneScriptingDoubleChannel
        default=80536.8366
        thinned_from=720
        keys [[0,80536.8366],[518.64,79470.0307],[1025.04,79470.0307],[1028.04,80566.3198],[1386.36,79185.5],[8265.66,79285.5651],[8268.96,80566.3198],[8654.16,79324.9223],[11063.58,78579.3113],[11842.74,79171.0225],[12417.3,79171.0225],[12420.84,80566.3198],[12862.14,79968.8518],[12877.86,79766.2555],[12916.2,79436.6384],[13266.9,79524.7571],[13270.08,80566.3198],[13604.64,79211.6251],[15984.42,78973.062],[19663.08,79241.753]]
       }
       Location.Z {
        type=MovieSceneScriptingDoubleChannel
        default=3110.5047
        thinned_from=594
        keys [[0,3035.5047],[56.16,3035.5047],[59.1,3146.9297],[62.28,3143.1772],[65.28,3131.1713],[68.28,3040.3598],[81.3,3037.1593],[229.2,3063.3446],[232.44,3100.9892],[236.76,3138.8877],[244.86,3188.2064],[249.06,3270.5693],[252.54,3301.3864],[255.96,3316.6139],[259.32,3322.3352],[263.16,3317.5372],[267.12,3301.1364],[274.08,3244.6133],[278.04,3220.7543],[283.26,3180.7333],[286.26,3117.1423],[289.26,3069.5842],[304.92,3070.7181],[310.92,3130.3574],[320.64,3199.8661],[323.88,3246.4677],[327.12,3263.8982],[330.24,3273.1148],[336.18,3271.399],[339.18,3264.4697],[349.5,3205.2848],[355.02,3153.8093],[358.08,3078.8355],[361.38,3070.7122],[373.8,3070.7122],[377.22,3101.018],[389.58,3222.3557],[392.52,3246.6947],[395.7,3243.3921],[398.7,3247.388],[407.52,3296.0534],[410.52,3307.8265],[424.86,3317.2785],[440.4,3313.801],[446.34,3301.0733],[464.34,3297.7494],[471.36,3291.0704],[489.78,3311.9695],[494.22,3306.6095],[498.48,3286.5126],[502.5,3252.1147],[506.28,3207.6773],[509.34,3154.5779],[513.06,3103.0238],[518.64,3070.3426],[1025.04,3070.3426],[1028.04,3039.3153],[1034.1,3041.2206],[1039.5,3035.5046],[1042.8,3039.3153],[1344.72,3067.21],[1353.36,3156.2473],[1360.44,3209.1309],[1364.28,3283.2859],[1368.06,3308.9882],[1371.72,3323.0299],[1378.62,3326.0078],[1382.52,3303.0682],[1386.36,3292.0241],[6471.72,3294.8159],[6478.98,3301.2579],[6482.34,3343.5553],[6486.9,3375.5427],[6490.14,3419.4791],[6494.82,3450.8701],[6497.76,3485.3672],[6505.38,3505.1514],[6510.18,3500.6872],[6518.1,3454.1605],[6521.28,3404.0371],[6525.6,3360.2967],[6528.42,3288.3028],[6532.8,3232.7951],[6535.8,3135.3748],[6540.72,3073.9982],[6568.02,3077.4457],[6601.26,3081.1845],[6606.24,3089.945],[6614.76,3076.7186],[6740.4,3077.1363],[6751.86,3088.058],[7345.62,3088.058],[7360.92,3073.2177],[7378.02,3073.2177],[7388.76,3099.0645],[7424.88,3095.792],[7432.08,3083.9333],[7487.7,3088.4958],[8265.66,3088.4958],[8268.96,3039.3153],[8346.48,3036.9372],[8394.24,3047.3004],[8612.1,3065.7405],[8616.3,3111.7812],[8622,3160.0742],[8627.94,3226.0687],[8631.54,3285.6215],[8635.26,3308.7215],[8638.86,3321.8814],[8642.94,3324.4452],[8646.78,3315.2553],[8654.16,3309.7079],[10707.48,3294.4822],[10713.78,3279.0305],[10719.36,3232.9086],[10722.78,3166.0972],[10731.84,3072.3872],[10789.44,3072.9916],[10792.86,3099.4241],[10798.02,3102.1038],[10810.26,3086.6607],[10815.24,3095.658],[10957.26,3096.4487],[10965.42,3093.8946],[10975.26,3077.8416],[10985.1,3075.6395],[10989,3091.7271],[10994.46,3092.8455],[10998.66,3081.9489],[11004.78,3073.998],[11014.8,3076.2281],[11024.34,3096.5733],[11036.4,3109.6238],[11042.58,3121.9464],[11059.56,3129.8662],[11063.58,3136.9803],[11511.78,3147.2474],[11535.24,3171.5494],[11598.72,3172.0394],[11638.62,3139.427],[11643.18,3131.6234],[11678.88,3100.5399],[11683.14,3093.0444],[11690.34,3089.0117],[11694.3,3075.9023],[12417.3,3073.3854],[12420.84,3039.3153],[12569.28,3049.2673],[12862.14,3061.194],[12866.52,3147.819],[12871.32,3197.6981],[12877.86,3252.572],[12881.46,3367.7478],[12886.38,3350.7225],[12890.1,3298.6902],[12893.94,3271.9974],[12901.68,3189.5431],[12906.48,3149.3095],[12912.36,3090.6447],[12916.2,3070.722],[13266.9,3069.7062],[13270.08,3039.3153],[13344.12,3040.1788],[13351.74,3036.1256],[13414.98,3045.7641],[13563.9,3067.0825],[13567.74,3110.3104],[13579.68,3211.8558],[13583.22,3284.96],[13586.58,3308.2553],[13589.76,3321.6236],[13593.18,3326.0633],[13600.44,3322.0872],[13604.64,3296.3983],[15728.04,3299.7923],[15735.66,3304.9834],[15744.3,3297.5119],[15747.48,3306.1617],[15752.58,3301.0006],[15764.94,3272.5515],[15770.22,3252.9084],[15773.58,3205.3126],[15778.5,3165.2724],[15781.62,3088.11],[15786.84,3073.1371],[15844.14,3073.7602],[15853.02,3094.9516],[15870.72,3087.6738],[15879.72,3097.49],[15885.84,3095.6839],[15895.5,3085.0215],[15899.04,3073.6907],[18663.9,3073.3259],[18667.8,3083.4345],[18673.5,3088.058],[19663.08,3088.058]]
       }
       Rotation.X {type=MovieSceneScriptingDoubleChannel, default=0, const [0,2]}
       Rotation.Y {type=MovieSceneScriptingDoubleChannel, default=0, const [0,2]}
       Rotation.Z {
        type=MovieSceneScriptingDoubleChannel
        default=-62.6264
        thinned_from=2154
        keys [[0,-152.6264],[101.64,-152.782],[105,-153.093],[108.3,-153.8707],[111.72,-154.3374],[121.38,-154.4929],[127.92,-154.9595],[150.72,-155.1151],[153.9,-155.7372],[160.44,-157.9148],[164.04,-158.2259],[176.88,-158.3814],[183.6,-159.6257],[196.32,-160.4034],[202.38,-160.4034],[205.74,-160.7145],[214.26,-160.8705],[219.66,-161.3428],[236.76,-161.7021],[252.54,-161.7021],[259.32,-160.7475],[315.18,-161.0585],[323.88,-160.5919],[417.18,-160.6803],[430.14,-160.1253],[446.34,-159.9698],[451.68,-159.6587],[478.2,-159.1921],[494.22,-159.1921],[502.5,-159.5031],[506.28,-159.9698],[685.62,-159.9698],[691.8,-159.1921],[1005.06,-159.1921],[1016.28,-150.7599],[1150.86,-150.7599],[1154.34,-149.8267],[1158.3,-149.2045],[1162.02,-147.4936],[1211.88,-147.1825],[1275.54,-147.3381],[1282.32,-148.5824],[1303.86,-150.2933],[1307.7,-150.2933],[1311.6,-152.1598],[1316.28,-156.9983],[1322.46,-159.9946],[1326.3,-160.3169],[1389.9,-160.3169],[1405.86,-162.8055],[1422.24,-163.7387],[1451.16,-163.5832],[1457.4,-162.8055],[1480.68,-162.65],[1483.56,-162.3389],[1489.08,-162.3389],[1497.78,-161.4056],[1512.54,-160.7835],[1527.3,-158.1393],[1530.36,-157.3616],[1533.48,-156.895],[1539.6,-155.4951],[1551.06,-153.9397],[1568.46,-152.6954],[1682.94,-153.162],[1688.58,-153.4731],[1720.02,-153.6287],[1725.78,-153.9397],[1748.82,-154.5619],[1754.58,-153.9397],[1757.46,-152.5399],[1760.4,-150.6734],[1763.28,-149.4291],[1769.34,-148.0292],[1775.1,-147.5626],[1850.7,-147.5626],[1854.24,-151.7622],[1875.9,-151.9177],[1889.16,-155.1841],[1892.76,-156.7394],[1895.94,-159.3836],[1899.06,-161.0946],[1904.52,-163.4277],[1907.82,-168.716],[1910.58,-170.2714],[1916.34,-172.6045],[1919.4,-174.6266],[1922.46,-176.1819],[1925.4,-178.204],[1937.64,-184.27],[1943.76,-186.2921],[1950.24,-189.0918],[1956.06,-190.6472],[1973.82,-192.2026],[2006.82,-194.2246],[2018.64,-194.069],[2039.64,-192.2026],[2060.34,-190.6472],[2078.58,-189.5584],[2090.28,-189.4028],[2096.34,-188.6252],[2105.22,-185.3588],[2111.4,-181.7814],[2117.22,-179.4483],[2126.1,-173.5378],[2129.28,-172.2935],[2138.16,-171.0491],[2143.74,-169.9604],[2164.2,-169.3382],[2175.72,-169.1826],[2178.54,-168.716],[2213.16,-168.2494],[2233.44,-167.0051],[2250.78,-166.3829],[2265.06,-165.1386],[2290.98,-163.2721],[2314.08,-163.4277],[2318.22,-163.8943],[2321.46,-164.672],[2331,-165.2942],[2339.88,-166.5385],[2363.1,-167.9383],[2397.42,-168.2494],[2403.36,-169.1826],[2429.82,-170.5825],[2441.22,-170.738],[2476.26,-172.449],[2531.52,-172.449],[2542.98,-173.3822],[2551.8,-173.6933],[2574.42,-173.5378],[2591.88,-172.6045],[2634.96,-171.9824],[2664.48,-171.9824],[2676.48,-172.9156],[2750.88,-173.2267],[2769.18,-174.1599],[2784,-174.0044],[2810.46,-172.449],[2813.4,-171.8268],[2819.22,-167.7828],[2825.28,-165.7608],[2837.16,-163.4277],[2840.04,-163.1166],[2845.86,-161.7167],[2851.92,-159.3836],[2863.68,-154.5619],[2866.68,-153.7842],[2875.56,-150.3623],[2887.5,-146.9404],[2890.38,-146.3183],[2893.32,-145.0739],[2899.2,-143.363],[2910.96,-141.4965],[2914.02,-140.5633],[2916.96,-139.1634],[2925.78,-136.6748],[2928.6,-136.2082],[2937.36,-136.2082],[2940.36,-137.1414],[2943.3,-138.3857],[2949.24,-141.8076],[2952.18,-143.0519],[2955.06,-145.2295],[2958.12,-148.3403],[2961.18,-152.851],[2967.3,-162.8055],[2973.66,-171.2047],[2979.6,-176.493],[2982.66,-179.6038],[2985.66,-183.3368],[2988.72,-188.003],[2991.84,-193.2914],[2994.96,-197.3354],[2997.96,-199.3574],[3004.14,-201.846],[3010.14,-205.7346],[3013.2,-207.2899],[3031.5,-220.0442],[3034.5,-221.133],[3048.96,-221.2885],[3057.72,-221.9107],[3077.52,-221.9107],[3080.46,-222.2218],[3083.7,-226.7325],[3087.12,-240.2644],[3090.54,-251.7744],[3093.96,-260.6402],[3097.26,-266.0841],[3100.32,-268.5727],[3103.32,-270.4392],[3106.5,-274.3277],[3112.68,-283.6601],[3115.68,-288.7929],[3122.16,-303.2581],[3125.4,-309.3242],[3137.76,-325.967],[3141,-330.9443],[3147.18,-338.5657],[3150.42,-339.499],[3153.66,-341.2099],[3156.54,-343.6985],[3171.78,-359.2525],[3174.96,-363.2966],[3177.9,-366.2518],[3188.22,-369.0516],[3191.28,-370.4514],[3194.46,-376.2064],[3197.7,-390.6716],[3209.28,-391.1382],[3212.64,-394.7157],[3215.94,-415.8691],[3219.18,-434.3784],[3222.42,-451.4878],[3225.78,-467.6639],[3229.2,-480.8848],[3238.92,-501.105],[3243.78,-509.8152],[3253.2,-523.9694],[3262.02,-529.5688],[3265.02,-532.3686],[3271.38,-533.9239],[3324.72,-530.5021],[3808.44,-525.2888],[4198.68,-532.6367],[4220.04,-534.3477],[4283.52,-536.3697],[6609.84,-535.2809],[6624.54,-536.9919],[6628.14,-538.0806],[6644.22,-539.325],[6662.1,-539.325],[6693.84,-541.6581],[7029.06,-541.8136],[7047.06,-544.9244],[7173.9,-543.1633],[7190.16,-543.0985],[7222.08,-539.508],[7230.06,-537.2675],[7233.24,-534.8886],[7280.94,-536.4674],[7352.82,-536.4674],[7364.46,-534.912],[7446.18,-537.2664],[7487.7,-540.0338],[7575.18,-538.2615],[8242.32,-538.2615],[8254.14,-510.7599],[8338.44,-510.7599],[8342.64,-512.3153],[8397.72,-512.4709],[8405.76,-516.0483],[8449.08,-517.7592],[8459.58,-519.6257],[8528.7,-519.7812],[8532.6,-521.1811],[8562.84,-521.5101],[8570.4,-523.8593],[8724.42,-522.5316],[8754,-520.043],[8763.42,-515.0657],[8781.66,-513.977],[8879.88,-511.1772],[8924.52,-511.7994],[8948.82,-519.2653],[8954.7,-522.6872],[8964,-525.4869],[8970.12,-528.1311],[9008.88,-528.5977],[9017.88,-530.7753],[9113.46,-532.6418],[9293.7,-529.0643],[9335.94,-529.2199],[9341.94,-529.9976],[9377.46,-529.9976],[9383.46,-528.9088],[9754.44,-532.9528],[10067.88,-530.9308],[10074.12,-524.3981],[10080.6,-519.8875],[10090.44,-510.0884],[10096.26,-507.4443],[10120.5,-501.5338],[10123.68,-499.3562],[10138.98,-482.5579],[10142.76,-479.2915],[10146.12,-474.6253],[10161.72,-460.4712],[10168.02,-453.9385],[10174.68,-448.8057],[10190.7,-430.763],[10197,-423.1416],[10216.44,-405.41],[10219.62,-403.2325],[10225.62,-400.4327],[10231.74,-395.1444],[10235.1,-391.567],[10238.16,-390.7893],[10244.1,-390.7893],[10256.4,-385.812],[10274.58,-381.4569],[10277.7,-379.4348],[10283.7,-377.1017],[10292.82,-369.9469],[10302.48,-360.3034],[10308.96,-352.3709],[10315.26,-341.9497],[10318.56,-337.2835],[10325.04,-329.0399],[10331.34,-322.9738],[10334.52,-317.0633],[10340.94,-303.8424],[10344.12,-298.243],[10354.14,-284.0888],[10360.62,-273.201],[10366.98,-266.6683],[10370.4,-262.1577],[10373.7,-256.7138],[10380,-248.3146],[10383.24,-244.7372],[10386.72,-241.6264],[10389.84,-237.5824],[10393.08,-234.9382],[10396.32,-231.3608],[10405.5,-226.2279],[10411.92,-221.7173],[10415.28,-220.1619],[10424.46,-214.4069],[10427.4,-211.9183],[10433.82,-208.1853],[10437,-206.941],[10443.18,-201.8082],[10449.3,-196.2087],[10452.42,-192.9424],[10455.42,-191.0759],[10459.2,-190.4537],[10464.3,-188.5873],[10467.54,-186.8763],[10472.28,-185.4764],[10478.46,-182.0546],[10483.08,-180.1881],[10486.26,-177.855],[10489.38,-176.7662],[10494,-176.144],[10500,-175.833],[10504.68,-175.2108],[10507.86,-173.4999],[10519.32,-173.3443],[10533.9,-171.6334],[10568.1,-170.8557],[10683.96,-171.2874],[10692.6,-172.6873],[10697.76,-172.6873],[10707.48,-174.0871],[10713.78,-174.5537],[10722.78,-175.6425],[10832.58,-176.8868],[10857.24,-176.7313],[10860.72,-176.2646],[10869,-175.798],[10905.84,-175.487],[10913.58,-175.0203],[10929.3,-174.7093],[10937.04,-174.2426],[10951.5,-171.9095],[10960.62,-168.9543],[10965.42,-168.3321],[10969.74,-167.2433],[10979.58,-166.3101],[11059.56,-166.4656],[11069.1,-166.9323],[11209.56,-167.2433],[11213.04,-167.71],[11221.32,-168.1766],[11234.34,-168.3321],[11237.7,-168.6432],[11242.8,-168.6432],[11246.22,-169.1098],[11251.38,-169.2654],[11255.04,-169.8875],[11271.9,-170.8208],[11295.18,-171.2874],[11313,-175.0203],[11322.06,-175.6425],[11325.6,-176.2646],[11334.24,-177.0424],[11365.44,-177.1979],[11374.8,-178.2867],[11383.86,-178.5978],[11437.08,-178.5978],[11440.56,-179.2199],[11495.46,-178.7533],[11511.78,-177.9756],[11530.86,-177.6645],[11573.7,-177.6645],[11578.44,-177.9756],[11588.88,-180.9308],[11604.66,-181.7086],[11884.2,-182.7973],[11889.84,-182.4863],[12395.82,-182.4863],[12409.26,-150.7599],[12481.8,-150.7599],[12489.6,-155.2706],[12501,-155.2706],[12504.78,-154.804],[12508.68,-151.8487],[12512.46,-150.2933],[12524.1,-149.8267],[12561,-149.8267],[12564.9,-151.071],[12569.28,-159.0036],[12573.36,-163.0476],[12577.32,-163.3587],[12585.18,-162.581],[12588.96,-161.9588],[12592.8,-160.0923],[12605.76,-159.9368],[12609.54,-159.6257],[12613.92,-157.4482],[12617.64,-157.1371],[12621.42,-158.0703],[12625.02,-158.3814],[12628.74,-157.9148],[12632.52,-159.3146],[12664.32,-159.3146],[12667.86,-159.6257],[12674.82,-159.3146],[12678.48,-156.6705],[12682.02,-161.1811],[12688.98,-161.1811],[12692.52,-159.1591],[12696.3,-159.4702],[12703.38,-159.4702],[12707.04,-165.5363],[12728.28,-165.6918],[12733.38,-166.3157],[12770.22,-166.498],[12774.42,-165.0837],[12778.68,-161.6448],[12782.52,-160.0265],[12921.96,-160.182],[12925.5,-160.9597],[13248.66,-160.9597],[13260.66,-150.7599],[13414.98,-150.7599],[13419.3,-150.9155],[13423.32,-152.9375],[13428.78,-153.5597],[13435.74,-155.5817],[13439.16,-156.0483],[13452.84,-156.3594],[13459.2,-159.0036],[13469.46,-160.7145],[13487.88,-161.026],[13494.12,-163.3843],[13501.56,-163.7053],[13504.98,-163.1858],[13655.94,-163.4974],[13672.5,-161.6309],[13675.92,-160.231],[13679.28,-159.2978],[13682.76,-158.9867],[13695.66,-156.3425],[13702.02,-154.3205],[13705.2,-153.5428],[13721.1,-151.2097],[13730.82,-150.5875],[13746.9,-150.432],[13753.38,-150.1209],[13769.28,-149.8098],[13816.14,-150.1209],[13831.44,-151.5208],[13841.1,-151.5208],[13850.58,-152.1429],[13866.06,-152.454],[13869.12,-152.7651],[13986.6,-152.9207],[14005.14,-154.9427],[14008.32,-155.0982],[14021.22,-157.1202],[14027.28,-158.6756],[14033.88,-163.1863],[14037,-164.4306],[14046.36,-167.0748],[14055.36,-169.0968],[14065.08,-173.1408],[14080.8,-174.0741],[14087.04,-175.6295],[14093.16,-176.2516],[14106,-178.2737],[14109.36,-178.4292],[14118.66,-180.1401],[14124.84,-180.4512],[14131.08,-182.6288],[14137.08,-182.9399],[14146.14,-182.9399],[14164.68,-180.7623],[14180.22,-177.9626],[14186.4,-177.3404],[14208.3,-172.5187],[14214.78,-170.6522],[14224.32,-166.9193],[14230.44,-165.5194],[14242.62,-163.8084],[14248.5,-163.3418],[14284.86,-163.3418],[14293.74,-164.2751],[14329.74,-164.5862],[14335.98,-165.2083],[14414.82,-165.3638],[14420.94,-165.8304],[14456.34,-165.6749],[14459.4,-165.3638],[14464.44,-165.3638],[14467.62,-165.0528],[14490.18,-165.2083],[14501.7,-166.6082],[14547.6,-166.6082],[14566.2,-165.3638],[14599.92,-164.7417],[14636.64,-164.1195],[14648.76,-163.4974],[14676.9,-162.5641],[14686.08,-162.5641],[14704.5,-161.6309],[14726.22,-161.6309],[14735.58,-162.2531],[14763.6,-163.0307],[14772.84,-163.8084],[14790.84,-164.1195],[14881.68,-164.1195],[14900.88,-162.8752],[14937.42,-163.0307],[14940.54,-164.1195],[14954.58,-164.4306],[15007.38,-164.4306],[15016.98,-163.0307],[15024.9,-162.7197],[15028.32,-156.6536],[15033.84,-149.0321],[15037.44,-139.8553],[15042.66,-134.878],[15045.96,-129.7452],[15050.82,-130.0563],[15054,-134.2558],[15059.52,-133.6337],[15063.12,-124.6124],[15068.4,-121.8126],[15071.76,-119.4796],[15076.8,-117.9241],[15079.92,-115.591],[15093.12,-112.6358],[15102.54,-108.2807],[15106.14,-105.0143],[15111.3,-103.4589],[15114.42,-101.5924],[15119.52,-100.1926],[15122.82,-97.3929],[15127.92,-95.2153],[15131.4,-91.7934],[15136.62,-88.8382],[15139.86,-84.483],[15144.72,-81.5278],[15148.2,-75.6173],[15153.48,-70.9511],[15157.2,-65.5072],[15162.12,-60.841],[15165.54,-55.8637],[15170.34,-53.5306],[15173.4,-51.6641],[15178.74,-49.7976],[15182.16,-46.5313],[15187.38,-44.8203],[15190.92,-40.3097],[15196.02,-38.1321],[15199.14,-35.1769],[15202.26,-33.4659],[15205.32,-32.5327],[15208.5,-31.2884],[15211.56,-30.1996],[15214.86,-29.1108],[15218.04,-27.8665],[15221.04,-26.4666],[15224.04,-24.7557],[15227.1,-22.8892],[15230.16,-21.4893],[15233.52,-19.7784],[15236.82,-17.4453],[15240,-14.9566],[15243.12,-14.0234],[15246.18,-12.9346],[15249.42,-11.3792],[15252.42,-10.1349],[15255.54,-8.8906],[15258.54,-7.4907],[15261.72,-5.6243],[15264.66,-3.6022],[15267.78,-0.3359],[15271.02,2.4638],[15274.26,5.8857],[15277.62,9.6187],[15281.04,13.1961],[15284.16,15.3736],[15287.28,18.0178],[15290.64,22.2174],[15294.06,27.6613],[15297.6,33.7274],[15301.02,38.3936],[15308.04,47.2594],[15311.28,51.6145],[15314.58,56.2807],[15317.88,62.0356],[15321.42,69.1905],[15324.96,77.5896],[15328.38,83.6557],[15331.86,88.1664],[15335.04,90.0329],[15338.22,92.9881],[15341.58,95.4767],[15344.94,98.432],[15348.42,102.165],[15355.14,109.0087],[15358.44,113.5194],[15361.74,116.4747],[15365.1,118.8078],[15368.46,122.3852],[15371.82,125.6515],[15378.48,133.273],[15381.84,137.4725],[15388.32,144.0052],[15391.56,147.7382],[15395.04,152.0933],[15401.1,155.9818],[15406.68,160.8036],[15411.9,163.4477],[15415.2,167.3362],[15420.24,168.425],[15423.72,171.8469],[15432.18,175.5799],[15437.28,176.0465],[15440.58,178.5351],[15449.1,180.0905],[15454.02,180.5571],[15457.56,183.6679],[15462.54,184.9123],[15465.66,184.9123],[15473.64,187.0898],[15486.9,187.2454],[15490.2,186.9343],[15495.36,183.5124],[15511.74,183.5124],[15515.1,185.3789],[15536.52,186.6232],[15623.1,187.4009],[15634.5,188.0231],[15723.06,188.1786],[15731.22,187.2454],[15735.66,187.0898],[15739.14,186.4677],[15761.34,185.8455],[15764.94,185.2233],[15821.82,185.0678],[15825.72,183.979],[15844.14,181.957],[15867.24,181.6459],[15879.72,180.7127],[15895.5,180.4016],[15908.04,179.4684],[15915.12,179.4684],[15918.9,179.1573],[15966.06,178.6907],[16132.8,178.5726],[16141.38,177.9821],[16198.14,177.671],[16531.68,178.4487],[16537.02,178.9153],[16656.78,178.2932],[16786.8,177.671],[17015.1,176.5822],[17195.46,177.36],[17808.24,176.4267],[17816.34,175.9601],[17840.58,175.649],[17990.4,175.649],[17993.64,176.1156],[18002.4,176.4267],[18154.02,176.5822],[18167.58,177.2044],[18263.52,177.2044],[18269.4,177.5155],[18279.24,177.5155],[18283.5,177.8266],[18418.98,178.1376],[18439.44,177.5155],[18484.32,176.8933],[18510.42,175.9601],[18568.68,176.1156],[18571.86,175.3379],[18597.36,174.0936],[18686.1,174.4047],[18695.52,174.8713],[18769.2,175.4935],[18850.62,175.4935],[18859.2,175.9601],[18862.32,176.4267],[19042.26,176.2712],[19047.18,176.5822],[19231.08,176.5822],[19243.08,177.671],[19284.96,177.671],[19290.36,178.097],[19294.14,179.3506],[19312.74,179.4202],[19328.04,178.8757],[19331.7,179.2787],[19412.52,178.5488],[19421.34,179.1689],[19424.82,179.8597],[19433.7,180.9018],[19439.04,180.9018],[19447.56,181.8719],[19456.26,183.2979],[19459.38,184.2449],[19472.64,185.3092],[19476,186.136],[19484.7,186.2357],[19490.1,184.4571],[19493.4,181.9702],[19498.62,181.7582],[19502.16,178.4519],[19507.26,176.6753],[19516.44,176.1054],[19520.4,169.379],[19526.04,168.7768],[19535.28,161.2151],[19538.82,153.8853],[19544.46,152.6013],[19547.94,154.9951],[19553.28,156.417],[19557,164.7465],[19562.52,166.5305],[19566.12,170.9935],[19572.12,175.3661],[19575.42,178.3293],[19589.82,178.8282],[19612.02,178.7277],[19618.14,178.2315],[19663.08,178.0888]]
       }
       Scale.X {type=MovieSceneScriptingDoubleChannel, default=1, const [1,2]}
       Scale.Y {type=MovieSceneScriptingDoubleChannel, default=1, const [1,2]}
       Scale.Z {type=MovieSceneScriptingDoubleChannel, default=1, const [1,2]}
      ]
     }]
   }
  ]
 }
 SK_Assault_Rifle_Mark_II_Base {
  id=#7
  possessed_class=FortSkeletalMeshActor
  tracks [
   {type=MovieSceneSpawnTrack, sections [{type=MovieSceneSpawnSection, channels [None {type=MovieSceneScriptingBoolChannel, default=true}]}]}
   {type=MovieSceneSkeletalAnimationTrack, sections []}
   {type=MovieScene3DAttachTrack, sections [{type=MovieScene3DAttachSection, end_frame=9807, start_frame=-10208}]}
   {
    type=MovieScene3DTransformTrack
    sections [{
      type=MovieScene3DTransformSection
      end_frame=9807
      start_frame=-10208
      channels [
       Location.X {type=MovieSceneScriptingDoubleChannel, default=0, const [0,1]}
       Location.Y {type=MovieSceneScriptingDoubleChannel, default=0, const [0,1]}
       Location.Z {type=MovieSceneScriptingDoubleChannel, default=0, const [0,1]}
       Rotation.X {type=MovieSceneScriptingDoubleChannel, default=0, const [0,1]}
       Rotation.Y {type=MovieSceneScriptingDoubleChannel, default=0, const [0,1]}
       Rotation.Z {type=MovieSceneScriptingDoubleChannel, default=0, const [0,1]}
       Scale.X {type=MovieSceneScriptingDoubleChannel, default=1, const [1,1]}
       Scale.Y {type=MovieSceneScriptingDoubleChannel, default=1, const [1,1]}
       Scale.Z {type=MovieSceneScriptingDoubleChannel, default=1, const [1,1]}
      ]
     }]
   }
   {
    type=MovieSceneVisibilityTrack
    sections [{
      type=MovieSceneVisibilitySection
      end_frame=9807
      start_frame=-1526
      channels [None {type=MovieSceneScriptingBoolChannel, default=true, keys [[-1525.5775,true],[8426.98,false],[8934.7625,true]]}]
     }]
   }
  ]
 }
 PistolCanary_EditorOnly_Cam {
  id=#4
  possessed_class=CineCameraActor
  tracks [
   {
    type=MovieScene3DTransformTrack
    sections [{
      type=MovieScene3DTransformSection
      channels [
       Location.X {type=MovieSceneScriptingDoubleChannel, default=-31.9492, const [-54.1385,2]}
       Location.Y {type=MovieSceneScriptingDoubleChannel, default=-259.6638, const [-231.1443,2]}
       Location.Z {type=MovieSceneScriptingDoubleChannel, default=155.3734, const [142.9567,2]}
       Rotation.X {type=MovieSceneScriptingDoubleChannel, default=0.1909, const [0.1909,2]}
       Rotation.Y {type=MovieSceneScriptingDoubleChannel, default=-0.8139, const [2.1861,2]}
       Rotation.Z {type=MovieSceneScriptingDoubleChannel, default=89.8906, const [91.0906,2]}
       Scale.X {type=MovieSceneScriptingDoubleChannel, default=1, const [1,2]}
       Scale.Y {type=MovieSceneScriptingDoubleChannel, default=1, const [1,2]}
       Scale.Z {type=MovieSceneScriptingDoubleChannel, default=1, const [1,2]}
      ]
     }]
   }
   {type=MovieSceneSpawnTrack, sections [{type=MovieSceneSpawnSection, channels [None {type=MovieSceneScriptingBoolChannel, default=true}]}]}
   {type=MovieScene3DAttachTrack, sections [{type=MovieScene3DAttachSection}]}
  ]
 }
 PistolCanary_EditorOnly_Player_Mannequin {
  id=#8
  possessed_class=FortPlayerMannequin
  tracks [
   {type=MovieSceneSpawnTrack, sections [{type=MovieSceneSpawnSection, channels [None {type=MovieSceneScriptingBoolChannel, default=true}]}]}
   {type=MovieSceneSkeletalAnimationTrack, sections [{type=MovieSceneSkeletalAnimationSection, end_frame=10274, start_frame=3480}]}
   {
    type=MovieScene3DTransformTrack
    sections [{
      type=MovieScene3DTransformSection
      end_frame=10274
      start_frame=3480
      channels [
       Location.X {
        type=MovieSceneScriptingDoubleChannel
        default=14082.1269
        thinned_from=139
        keys [[3480,14347.2861],[5221.1775,14441.1479],[5233.06,14493.8075],[5236.6,14583.2221],[5247.58,14637.9739],[5435.2575,14558.2102],[6250.7175,14600.2384],[6306.16,14914.7718],[6483.4575,14943.715],[7955.08,14921.991],[8042.3775,14431.5634],[8094.64,14375.5219],[8105.86,14334.3137],[10274,14322.2202]]
       }
       Location.Y {
        type=MovieSceneScriptingDoubleChannel
        default=81558.0434
        thinned_from=139
        keys [[3480,81195.4561],[5435.2575,79308.8613],[6483.4575,79470.5779],[10274,79522.0181]]
       }
       Location.Z {
        type=MovieSceneScriptingDoubleChannel
        default=4337.0708
        thinned_from=86
        keys [[3480,4279.0171],[5195.62,4272.9224],[5207.4375,4253.2715],[5214.46,4212.6574],[5221.1775,4137.6307],[5224.78,4086.5895],[5233.06,4032.2601],[5236.6,3918.4908],[5240.44,3882.0324],[5298.28,3877.5668],[5328.6375,3893.0507],[5332.5975,3943.9173],[5344.2975,3914.3064],[5351.62,3877.6143],[5366.62,4022.9372],[5370.28,4053.3102],[5374.12,4072.4195],[5377.72,4081.1533],[5381.02,4078.8606],[5387.56,4049.1092],[5391.0975,4021.4205],[5394.9375,3982.4103],[5398.7175,3928.9385],[5402.3775,3865.0878],[5405.86,3793.1395],[5413,3628.4435],[5416.7175,3523.8494],[5420.38,3408.9362],[5424.1,3285.0721],[5427.5175,3148.3753],[5431.7175,3088.058],[6270.1575,3088.8986],[6274.06,3073.2854],[7941.7,3070.925],[7959.8175,3089.1724],[10274,3088.058]]
       }
       Rotation.X {type=MovieSceneScriptingDoubleChannel, default=0, const [0,3]}
       Rotation.Y {type=MovieSceneScriptingDoubleChannel, default=0, const [0,3]}
       Rotation.Z {
        type=MovieSceneScriptingDoubleChannel
        default=-80.3855
        thinned_from=378
        keys [[3480,-179.713],[4713.34,-179.713],[4716.6375,-179.4019],[4720.3,-175.8245],[4723.5975,-168.6697],[4727.3775,-164.6257],[4737.7575,-160.8927],[4836.8775,-161.0482],[4847.0175,-161.9815],[4853.98,-162.9147],[4874.98,-164.6257],[4882.06,-166.0255],[4899.46,-171.1583],[4913.32,-173.4914],[4921.1775,-174.1136],[4932.28,-176.9133],[4936.1775,-178.1577],[4939.48,-179.713],[4950.4,-180.8018],[5007.4575,-179.8686],[5267.62,-179.8686],[5275.3,-180.8018],[5298.28,-182.0461],[5394.9375,-182.2017],[5413,-183.446],[5471.02,-183.2905],[5478.28,-182.6683],[5545.1775,-182.5128],[5555.0175,-182.0461],[5571.3975,-182.2017],[5585.08,-183.6015],[5592.1575,-183.9126],[5974.84,-184.5405],[6005.0175,-183.1234],[6204.2175,-183.0053],[6211.7175,-184.2497],[6215.68,-184.5607],[6285.28,-184.7163],[6289.78,-185.8051],[6293.98,-187.2049],[6298.06,-187.8271],[6306.16,-188.2937],[6314.92,-189.538],[6323.32,-189.538],[6339.8775,-188.4492],[6389.3775,-187.6715],[6393.2175,-187.0494],[6483.4575,-187.0494],[6491.02,-187.516],[6577.96,-187.516],[6582.7575,-186.5828],[6738.76,-185.9606],[6742.42,-186.2717],[6770.5,-186.7383],[7800.28,-186.7383],[7813.7175,-188.4492],[7823.8575,-188.6048],[7858.66,-187.0494],[7883.6775,-186.5828],[7968.4,-186.4272],[7972.9,-185.9606],[7995.5175,-185.6495],[8018.8575,-183.7831],[8046.9975,-183.3164],[8085.46,-182.2276],[8105.86,-181.9166],[8471.38,-180.3611],[9012.0975,-180.125],[9018.88,-181.0582],[9026.2,-181.0582],[9038.1375,-179.8139],[9063.7,-179.5028],[9067.48,-179.8139],[9073.96,-179.8139],[9084.5175,-181.0582],[9091.3,-181.3693],[9111.5775,-181.6804],[9114.9375,-181.9915],[9156.34,-181.9915],[9160.06,-182.3026],[9177.4575,-182.3026],[9184.06,-182.6136],[9227.8,-182.3026],[9258.6375,-180.5916],[9273.58,-179.0362],[9276.9375,-179.0362],[9283.54,-180.2805],[9290.2,-180.2805],[9293.8,-180.5916],[9304.18,-184.0135],[9307.6575,-185.4133],[9311.08,-187.2798],[9318.58,-189.3018],[9333.4,-191.1683],[9382.12,-190.5461],[9408.4575,-189.3018],[9424.54,-188.9908],[9457.48,-186.191],[9470.68,-185.8799],[9486.5775,-184.9467],[9591.5775,-182.7692],[10218.6375,-182.7692],[10274,-182.3011]]
       }
       Scale.X {type=MovieSceneScriptingDoubleChannel, default=1, const [1,3]}
       Scale.Y {type=MovieSceneScriptingDoubleChannel, default=1, const [1,3]}
       Scale.Z {type=MovieSceneScriptingDoubleChannel, default=1, const [1,3]}
      ]
     }]
   }
  ]
 }
 FortPlayerMannequin2 {
  id=#9
  possessed_class=FortPlayerMannequin
  tracks [
   {type=MovieSceneSpawnTrack, sections [{type=MovieSceneSpawnSection, channels [None {type=MovieSceneScriptingBoolChannel, default=true}]}]}
   {type=MovieSceneSkeletalAnimationTrack, sections [{type=MovieSceneSkeletalAnimationSection, end_frame=9758, start_frame=3480}]}
   {
    type=MovieScene3DTransformTrack
    sections [{
      type=MovieScene3DTransformSection
      end_frame=9758
      start_frame=3480
      channels [
       Location.X {type=MovieSceneScriptingDoubleChannel, default=11713.9543, const [12339.1272,2]}
       Location.Y {type=MovieSceneScriptingDoubleChannel, default=77710.751, const [76917.411,2]}
       Location.Z {type=MovieSceneScriptingDoubleChannel, default=3072.0217, const [3316.4019,2]}
       Rotation.X {type=MovieSceneScriptingDoubleChannel, default=0, const [0,2]}
       Rotation.Y {type=MovieSceneScriptingDoubleChannel, default=0, const [0,2]}
       Rotation.Z {
        type=MovieSceneScriptingDoubleChannel
        default=-51.0742
        thinned_from=420
        keys [[3480,-140.9186],[3867.3375,-140.9186],[3882.3975,-137.9633],[3939.7575,-138.1189],[3950.1375,-138.43],[3972.76,-139.6743],[3985.1175,-139.8298],[4251.22,-139.8298],[4297.8375,-137.8078],[4336.24,-137.9633],[4349.4375,-137.4967],[4375.3575,-137.0301],[4384.24,-136.408],[4392.6375,-136.408],[4396.96,-136.0969],[4513.78,-136.0969],[4527.0375,-135.4747],[4570.6,-135.4747],[4578.46,-135.9413],[4590.5775,-136.2524],[5175.82,-136.2524],[5185.9,-137.0301],[5206.7175,-137.4967],[5243.44,-137.9633],[5507.38,-137.9633],[5511.7575,-138.5855],[5530,-138.1189],[5534.9175,-137.8078],[5539.24,-137.8078],[5547.9375,-137.1856],[5560.78,-135.0081],[5569.36,-133.2971],[5578.12,-132.5194],[5587.24,-132.3639],[5592.16,-132.0528],[5596.5375,-132.0528],[5601.0975,-131.7418],[5618.38,-129.2531],[5622.6375,-128.942],[5639.02,-126.6089],[5652.46,-122.4094],[5665.1175,-121.2766],[5672.74,-121.1585],[5690.26,-119.7415],[5694.34,-119.6234],[5702.38,-119.0329],[5710.6575,-118.7967],[5714.6775,-118.2063],[5727.16,-118.0882],[5736.16,-117.6158],[5752,-117.2615],[5831.9775,-117.2615],[5865.04,-117.9701],[5881.78,-119.2691],[5903.92,-119.8595],[5908.54,-120.45],[5912.98,-120.6862],[5921.92,-121.8671],[5927.2575,-123.7565],[5931.82,-126.945],[5940.6375,-129.7792],[5945.08,-131.6686],[5949.52,-132.9676],[5953.72,-133.44],[5963.26,-133.5581],[5977.1175,-134.8571],[5981.44,-135.5656],[5986.06,-136.7465],[5995,-137.455],[5999.32,-138.7541],[6007.9,-139.3445],[6020.5575,-140.8797],[6029.4375,-143.5958],[6033.7,-145.131],[6042.28,-146.6661],[6049.4775,-147.1385],[6053.5,-147.847],[6074.2,-148.3194],[6102.9975,-148.3194],[6107.14,-148.5556],[6128.26,-148.5556],[6136.8975,-149.5003],[6162.22,-149.8546],[6170.5575,-150.445],[6206.9775,-151.6259],[6215.4975,-152.5707],[6223.9575,-152.925],[6265.7175,-152.925],[6273.9975,-153.1611],[6294.6975,-155.1687],[6373.36,-155.2868],[6385.48,-156.0645],[6389.8,-156.9977],[6394.12,-157.1533],[6398.38,-157.9309],[6402.58,-159.4863],[6411.04,-160.1085],[6427.4775,-160.264],[6444.28,-159.3308],[6497.44,-159.3308],[6501.7,-159.0197],[6637.96,-159.0197],[6646.24,-158.0865],[6671.08,-157.4643],[7429.5375,-156.9977],[7855.6575,-156.8422],[7864.36,-156.3755],[8633.8,-158.8642],[8637.28,-159.1753],[8651.62,-159.4863],[8693.68,-163.8415],[8697.3975,-164.4636],[8730.4,-164.4636],[8741.38,-163.8415],[8774.56,-163.2193],[8825.74,-161.6639],[8936.7975,-161.3528],[8944.72,-162.7527],[8963.8,-163.6859],[8975.38,-163.6859],[8986.7175,-162.4416],[8998.36,-159.9529],[9002.14,-159.7974],[9006.16,-158.7086],[9009.88,-157.1533],[9013.8375,-156.5311],[9022.4175,-156.0732],[9026.3175,-155.084],[9030.46,-154.6117],[9038.0775,-154.1393],[9045.6375,-153.3127],[9056.9775,-152.9584],[9076.78,-152.8403],[9080.4975,-152.4861],[9103.66,-152.2499],[9111.4,-151.6594],[9146.14,-150.2423],[9153.7,-150.2423],[9161.56,-149.8881],[9184.72,-149.8881],[9188.8,-150.5966],[9196.5375,-153.5489],[9200.1375,-155.6745],[9211.42,-159.8077],[9214.8975,-161.3429],[9222.6975,-162.6419],[9256.3,-162.6419],[9259.8975,-163.5751],[9264.0375,-167.9302],[9268.12,-175.3961],[9272.02,-178.0403],[9279.94,-178.0403],[9283.8375,-177.7292],[9288.52,-166.6859],[9292.24,-158.7396],[9299.4375,-158.1375],[9303.1,-158.4918],[9306.9375,-161.7983],[9310.78,-166.0496],[9314.4375,-169.4742],[9318.04,-169.7104],[9325.96,-169.7104],[9329.74,-171.3637],[9333.34,-171.9541],[9344.5,-171.9541],[9348.2175,-170.8913],[9352,-167.939],[9355.6,-167.1124],[9362.74,-166.8762],[9385.0575,-168.7657],[9388.9,-168.6476],[9395.98,-167.4667],[9500.1375,-167.4667],[9504.52,-168.6476]]
       }
       Scale.X {type=MovieSceneScriptingDoubleChannel, default=1, const [1,2]}
       Scale.Y {type=MovieSceneScriptingDoubleChannel, default=1, const [1,2]}
       Scale.Z {type=MovieSceneScriptingDoubleChannel, default=1, const [1,2]}
      ]
     }]
   }
  ]
 }
 FortPlayerMannequin4 {
  id=#10
  possessed_class=FortPlayerMannequin
  tracks [
   {type=MovieSceneSpawnTrack, sections [{type=MovieSceneSpawnSection, channels [None {type=MovieSceneScriptingBoolChannel, default=true}]}]}
   {type=MovieSceneSkeletalAnimationTrack, sections []}
   {type=MovieSceneSkeletalAnimationTrack, sections [{type=MovieSceneSkeletalAnimationSection, end_frame=9708, start_frame=3480}]}
   {
    type=MovieScene3DTransformTrack
    sections [{
      type=MovieScene3DTransformSection
      end_frame=9708
      start_frame=3480
      channels [
       Location.X {
        type=MovieSceneScriptingDoubleChannel
        default=18052.9628
        thinned_from=121
        keys [[3480,17835.3838],[4894.5375,17861.0744],[4898.98,17885.749],[5134.78,17765.7838],[6061.9,17735.5689],[6082,17669.1023],[6130.1775,17224.5363],[6196.4175,17007.8933],[6266.56,16374.9874],[6338.8,16227.034],[6363.52,16047.193],[6425.44,15459.0723],[7990.5975,15416.8538],[7999.2975,15397.3668],[8179.9,15362.6654],[8184.22,15378.1171],[9472.78,15378.1171]]
       }
       Location.Y {type=MovieSceneScriptingDoubleChannel, default=78038.5744, thinned_from=121, keys [[3480,77194.6867],[9472.78,77259.1714]]}
       Location.Z {
        type=MovieSceneScriptingDoubleChannel
        default=3545.15
        thinned_from=95
        keys [[3480,3470.6437],[6061.9,3470.3388],[6068.2575,3542.5263],[6072.8775,3603.4169],[6082,3643.3909],[6086.2575,3702.9706],[6090.58,3700.5159],[6094.6575,3633.4294],[6099.04,3576.2572],[6107.6775,3441.7849],[6112.06,3349.1416],[6116.56,3243.538],[6120.8175,3117.8945],[6125.56,2986.3531],[6130.1775,2895.7622],[6148.6,2930.5502],[6187.66,2928.8648],[6262.36,3029.2742],[6266.56,3041.1773],[6282.4575,3053.7751],[6352.2375,3066.1376],[6367.9575,3074.6489],[6376.72,3161.3585],[6386.14,3244.7399],[6390.4575,3269.9474],[6394.78,3278.5011],[6399.22,3272.7042],[6403.3,3251.3407],[6412,3176.1517],[6425.44,3216.725],[7986.46,3225.6115],[7999.2975,3233.8153],[8179.9,3227.55],[8184.22,3231.2344],[9472.78,3231.2344]]
       }
       Rotation.X {type=MovieSceneScriptingDoubleChannel, default=0, const [0,2]}
       Rotation.Y {type=MovieSceneScriptingDoubleChannel, default=0, const [0,2]}
       Rotation.Z {
        type=MovieSceneScriptingDoubleChannel
        default=-98.8217
        thinned_from=695
        keys [[3480,-214.3302],[3900.04,-214.3302],[3918.22,-211.375],[3954.1,-211.686],[3974.44,-217.441],[3986.56,-218.5298],[3997.06,-221.3295],[4007.5,-227.3956],[4018.3,-234.8615],[4029.6975,-240.772],[4052.98,-241.0831],[4064.62,-239.5277],[4112.1975,-239.5277],[4123,-240.3054],[4144.6,-243.2607],[4181.92,-244.505],[4281.52,-244.3494],[4306.72,-245.1271],[4328.5,-244.505],[4333.3,-244.0384],[4412.44,-243.8828],[4417.3,-243.4162],[4522.12,-243.4162],[4561.0575,-242.483],[4599.1,-242.483],[4607.86,-240.9276],[4625.9775,-239.8388],[4673.8,-241.2386],[4707.1575,-243.4162],[4711.4175,-244.0384],[4748.5,-244.1939],[4759.48,-243.7273],[4769.5,-239.5277],[4781.44,-231.4396],[4791.3375,-223.0405],[4804.06,-210.1306],[4814.38,-191.6214],[4824.16,-180.7336],[4834.84,-174.512],[4857.28,-159.4246],[4867.72,-138.7378],[4881.0375,-137.0268],[4886.14,-136.7158],[4912,-136.7158],[4916.68,-136.0936],[4921.3,-132.2051],[4926.16,-130.4941],[4930.5375,-129.2498],[4934.98,-128.4721],[4944.4,-128.3166],[4957.72,-128.9388],[4971.1,-129.0943],[4982.14,-132.3606],[4986.64,-132.9828],[4991.74,-133.2939],[4996.6,-134.3827],[5001.2775,-137.3379],[5006.38,-146.9814],[5013.7575,-159.1135],[5018.68,-164.8685],[5023.1775,-168.6014],[5028.16,-178.7115],[5040.7,-209.3529],[5045.5,-219.6186],[5059.42,-235.6392],[5063.9175,-238.1278],[5071.96,-240.772],[5080.3,-242.1719],[5110.5975,-242.6385],[5119.9575,-243.7273],[5130.0975,-244.3494],[5134.78,-245.1271],[5205.7575,-245.2827],[5216.38,-245.7493],[5232.52,-244.9716],[5244.28,-243.7273],[5249.8,-242.1719],[5255.74,-236.2614],[5261.74,-235.3281],[5290.72,-235.3281],[5302.54,-233.9283],[5370.16,-232.995],[5381.86,-232.5284],[5415.0975,-232.8395],[5434.2975,-233.4616],[5559.28,-233.934],[5563.72,-234.6425],[5572.48,-235.1149],[5606.32,-235.1149],[5610.4575,-237.1225],[5623.5375,-239.2481],[5632.2375,-239.8385],[5659.48,-240.1928],[5666.08,-242.4365],[5679.28,-243.027],[5688.16,-244.0898],[5729.08,-244.6803],[5746.96,-245.5069],[5767.72,-245.625],[5810.3175,-247.3963],[5895.88,-246.4516],[5946.82,-242.0823],[5954.9175,-241.8461],[5967.8775,-240.5471],[5998.1175,-239.8385],[6016.0575,-237.2405],[6042.58,-237.1599],[6061.9,-239.1819],[6068.2575,-240.8929],[6086.2575,-241.515],[6103.1775,-241.3595],[6107.6775,-240.5818],[6116.56,-237.1599],[6120.8175,-234.8268],[6134.56,-230.6272],[6139,-229.8495],[6148.6,-226.8943],[6153.0975,-226.5832],[6187.66,-226.5832],[6192.0375,-226.2721],[6200.74,-224.8722],[6221.86,-224.7167],[6230.9775,-222.8502],[6235.36,-222.5391],[6317.3775,-222.6947],[6325.8975,-223.7835],[6343.36,-223.939],[6352.2375,-224.5612],[6376.72,-224.5612],[6386.14,-223.939],[6390.4575,-223.3168],[6403.3,-223.3168],[6407.3775,-222.5391],[6412,-219.895],[6420.88,-218.0285],[6425.44,-215.5398],[6434.8,-212.2735],[6439.18,-212.2735],[6443.44,-211.8069],[6447.8775,-209.0072],[6456.82,-207.2962],[6461.1975,-205.7408],[6465.3975,-204.8076],[6469.5375,-204.4965],[6477.76,-204.4965],[6506.32,-202.0079],[6529.9575,-202.0079],[6537.8175,-200.4525],[6557.1975,-199.0526],[6754.72,-198.1194],[7068.3375,-198.1194],[7075.8975,-198.4736],[7103.5,-198.3555],[7107.4575,-198.946],[7111.48,-200.0088],[7115.3175,-200.3631],[7127.32,-200.245],[7131.4,-199.6545],[7135.06,-198.7098],[7158.0975,-197.4108],[7205.9775,-197.4108],[7231.78,-195.5214],[7243.1775,-195.2852],[7251.0375,-193.2776],[7284.94,-191.0339],[7300.2375,-191.5063],[7326.4,-191.7425],[7337.86,-192.6872],[7356.64,-193.1595],[7391.2575,-193.75],[7470.3975,-193.2776],[7481.5,-192.8053],[7507.48,-192.8053],[7526.38,-194.4586],[7537.84,-195.9937],[7571.56,-198.0013],[7626.76,-198.8279],[7726.54,-198.8279],[7730.44,-198.4736],[7738.8375,-196.4661],[7751.1975,-195.5214],[7786.48,-195.2852],[7823.02,-195.9937],[7878.6975,-195.3716],[7895.6775,-193.9717],[7899.8775,-193.9717],[7908.0375,-192.5718],[7932.58,-191.7942],[7944.8775,-191.7942],[7953.22,-190.8609],[7961.62,-189.15],[7986.46,-188.8389],[8085.8175,-188.8389],[8091.1,-190.7054],[8128.72,-190.8609],[8132.98,-191.6386],[8141.74,-192.2608],[8210.44,-193.0385],[8227.84,-194.7494],[8290.2375,-194.7494],[8298.0375,-194.2828],[8312.26,-193.9717],[8380.36,-193.9717],[8394.8775,-195.216],[8405.74,-195.0605],[8431.18,-192.1052],[8449.48,-190.8609],[8467.18,-191.0164],[8510.44,-194.1273],[8517.6375,-194.4383],[8546.2575,-194.4383],[8575.72,-192.4163],[8622.0375,-192.8829],[8647.06,-194.4383],[8675.02,-194.1273],[8726.44,-190.7054],[8755.06,-190.8609],[8775.94,-192.4163],[8811.2775,-192.5718],[8825.6775,-191.6386],[8843.2575,-192.1052],[8871.94,-192.4163],[8961.94,-192.2694],[8973.22,-191.4427],[9007.3,-192.1513],[9011.1975,-191.6789],[9018.28,-191.5608],[9044.26,-190.498],[9166.3575,-191.5608],[9207.8775,-189.5533],[9219.4,-189.6714],[9231.1,-193.8046],[9235.12,-195.694],[9239.14,-198.292],[9243.3375,-201.5985],[9247.66,-204.1965],[9251.7975,-207.1488],[9260.26,-210.5734],[9271.84,-211.0458],[9280.06,-213.6438],[9284.14,-213.9981],[9321.46,-213.9981],[9350.56,-209.1564],[9354.16,-207.8574],[9357.94,-205.4955],[9365.6775,-198.4101],[9369.5175,-194.395],[9373.1775,-191.9151],[9376.96,-190.7342],[9391.18,-190.6161],[9398.5,-189.6714],[9431.1975,-188.7267],[9455.3775,-188.7267],[9472.78,-188.1362]]
       }
       Scale.X {type=MovieSceneScriptingDoubleChannel, default=1, const [1,2]}
       Scale.Y {type=MovieSceneScriptingDoubleChannel, default=1, const [1,2]}
       Scale.Z {type=MovieSceneScriptingDoubleChannel, default=1, const [1,2]}
      ]
     }]
   }
  ]
 }
 FortPlayerMannequin3 {
  id=#11
  possessed_class=FortPlayerMannequin
  tracks [
   {type=MovieSceneSpawnTrack, sections [{type=MovieSceneSpawnSection, channels [None {type=MovieSceneScriptingBoolChannel, default=true}]}]}
   {type=MovieSceneSkeletalAnimationTrack, sections [{type=MovieSceneSkeletalAnimationSection, end_frame=9851, start_frame=3480}]}
   {
    type=MovieScene3DTransformTrack
    sections [{
      type=MovieScene3DTransformSection
      end_frame=9851
      start_frame=3480
      channels [
       Location.X {
        type=MovieSceneScriptingDoubleChannel
        default=9924.2094
        thinned_from=105
        keys [[3480,10615.8855],[6690.8175,10651.0898],[6707.3775,10694.9852],[6715.8975,10753.7874],[6719.62,10833.8989],[6726.82,10932.3982],[6747.8775,11220.8725],[6756.76,11266.5135],[6765.76,11361.965],[6769.8975,11457.3933],[6851.7975,12528.4124],[6877.7775,12697.8978],[6898.0575,12927.2655],[6907.8375,12991.2857],[6922.8375,13212.6951],[6927.2175,13243.0229],[7013.9175,13244.3977],[7044.7575,13405.713],[7128.1575,13449.3098],[7145.32,13551.0893],[7225.72,13582.9955],[7452.0375,13623.3355],[9411.16,13623.3355]]
       }
       Location.Y {
        type=MovieSceneScriptingDoubleChannel
        default=79131.4632
        thinned_from=105
        keys [[3480,78981.8342],[7452.0375,78104.3275],[9411.16,78104.3275]]
       }
       Location.Z {
        type=MovieSceneScriptingDoubleChannel
        default=4364.6238
        thinned_from=100
        keys [[3480,4289.6473],[6698.98,4290.8],[6707.3775,4340.3725],[6715.8975,4426.9975],[6719.62,4506.4587],[6726.82,4556.2669],[6733.96,4536.397],[6737.5,4515.1856],[6740.9175,4484.4319],[6747.8775,4397.8399],[6756.76,4337.7185],[6765.76,4179.6169],[6769.8975,3977.7654],[6773.6175,3848.7518],[6777.4575,3669.1536],[6781.6,3462.1785],[6785.4375,3284.0875],[6789.64,3105.7214],[6794.2,2995.8832],[6833.74,3004.5105],[6842.74,3022.0988],[6860.02,3025.2936],[6868.96,3016.9773],[6877.7775,3049.9104],[6882.82,3043.8849],[6892.4175,2997.8827],[6898.0575,3055.0581],[6907.8375,3110.0833],[6913.7175,3182.2792],[6922.8375,3255.5645],[6931.7775,3210.3712],[6987.1575,3212.8949],[7001.0775,3344.5807],[7005.46,3428.5148],[7013.9175,3451.8421],[7018.4175,3452.9848],[7022.86,3361.1739],[7026.9375,3312.8966],[7031.02,3238.3628],[7035.8775,3219.4856],[7309.0575,3208.7613],[9411.16,3207.2817]]
       }
       Rotation.X {type=MovieSceneScriptingDoubleChannel, default=0, const [0,2]}
       Rotation.Y {type=MovieSceneScriptingDoubleChannel, default=0, const [0,2]}
       Rotation.Z {
        type=MovieSceneScriptingDoubleChannel
        default=-52.1198
        thinned_from=579
        keys [[3480,-142.1198],[3587.9175,-142.2753],[3602.5575,-142.1198],[3610.2975,-142.5864],[3635.5575,-143.2086],[3678.9375,-143.5197],[3682.6575,-143.8307],[3719.7375,-144.1418],[3723.52,-144.4529],[3738.0975,-144.764],[3818.26,-144.9195],[3821.98,-144.764],[3887.9175,-144.764],[3891.2775,-144.6085],[4021.4175,-144.6085],[4054.12,-144.9195],[4460.5575,-144.9195],[4478.5575,-144.4529],[4543,-144.4529],[4633.9,-143.5197],[4704.1575,-143.6752],[4711.7775,-143.2086],[4715.5575,-142.742],[4719.34,-141.8087],[4723.18,-139.9423],[4728.04,-136.6759],[4731.82,-135.6457],[4735.6575,-135.1526],[4739.7375,-133.3813],[4743.9375,-131.9642],[4748.0175,-130.3109],[4751.92,-128.4214],[4755.94,-124.9968],[4767.6375,-118.8561],[4775.32,-118.5018],[4802.5,-118.2657],[4810,-118.5328],[4828.8375,-118.5328],[4832.74,-118.0662],[4863.6975,-118.0662],[4867.5975,-114.6443],[4871.4375,-109.8226],[4875.3975,-103.2899],[4879.5975,-91.78],[4883.74,-81.3588],[4887.6975,-74.3595],[4891.7175,-68.76],[4895.9775,-62.3829],[4899.8775,-60.3609],[4904.0775,-56.4724],[4908.2775,-47.7621],[4912.06,-46.5178],[4919.3775,-46.5178],[4923.0975,-46.8289],[4927.1775,-60.983],[4931.32,-90.8467],[4935.46,-118.2218],[4939.48,-130.1984],[4943.4975,-137.5087],[4952.14,-159.2843],[4956.5175,-172.0386],[4960.8975,-182.1487],[4964.9775,-187.4371],[4968.88,-191.17],[4972.7175,-192.5699],[4980.6975,-198.0138],[4984.9575,-201.2801],[4988.9175,-202.2134],[4996.7175,-203.1466],[5000.8,-204.3909],[5004.8775,-204.8575],[5016.76,-204.8575],[5020.8975,-205.6353],[5024.8575,-208.9016],[5028.52,-210.3015],[5032.3575,-210.7681],[5051.26,-210.7681],[5055.0975,-208.9016],[5058.82,-205.0131],[5066.8575,-187.5926],[5070.76,-182.1487],[5074.6575,-174.8383],[5078.4975,-167.9946],[5082.0975,-163.0173],[5085.8175,-159.9065],[5089.7775,-157.4178],[5093.4975,-155.7069],[5097.1575,-153.3738],[5104.5975,-151.8184],[5112.9375,-150.5741],[5116.9575,-150.263],[5124.5775,-150.263],[5128.4775,-149.6408],[5139.8775,-149.1742],[5151.3375,-148.0854],[5170,-147.6188],[5206.7175,-147.3077],[5210.32,-146.5301],[5225.2575,-145.1302],[5235.8175,-145.1302],[5243.26,-143.8859],[5247.0375,-143.5748],[5250.8775,-143.5748],[5258.3775,-142.7971],[5278,-142.3305],[5351.0775,-142.1749],[5373.1575,-142.9526],[5398.8375,-144.6636],[5468.1975,-144.6636],[5475.3975,-144.3525],[5479.2375,-143.7303],[5482.96,-142.486],[5486.7375,-141.7083],[5490.46,-141.2417],[5505.2175,-141.2417],[5509.18,-140.9306],[5516.8575,-139.5308],[5520.58,-138.5975],[5528.68,-137.5087],[5536.24,-136.1089],[5540.2575,-136.1089],[5552.1975,-135.0201],[5564.1375,-134.2424],[5567.9775,-133.6202],[5571.9975,-133.3091],[5579.32,-133.1536],[5586.7,-132.3759],[5597.86,-132.3759],[5601.5775,-132.0648],[5605.54,-131.2871],[5609.32,-130.976],[5616.8775,-130.976],[5620.54,-130.5094],[5624.4375,-130.3539],[5663.1375,-130.3539],[5674.8975,-129.7317],[5678.7975,-129.2651],[5698.1175,-128.6429],[5720.9775,-128.6429],[5736.9375,-125.2211],[5744.9775,-124.7545],[5764.18,-124.4434],[5771.8575,-124.7545],[5813.26,-124.91],[5824.6575,-125.0655],[5828.3175,-125.5322],[5843.7975,-125.9988],[5848.0575,-126.4654],[5851.9575,-127.5542],[5860,-128.1763],[5875.1775,-128.0208],[5882.7975,-127.5542],[5886.7,-127.5542],[5890.6575,-127.8652],[5898.6975,-129.4207],[5906.68,-132.2204],[5910.6975,-133.1536],[5914.7775,-133.6202],[5938.0575,-133.4647],[5953.3,-133.9313],[5968.4775,-135.4867],[6013.2375,-135.7978],[6016.7175,-136.731],[6020.6175,-137.0421],[6027.8775,-138.2864],[6032.0775,-139.6863],[6035.8575,-140.464],[6051.04,-140.9306],[6062.4375,-141.7083],[6069.9375,-143.2637],[6073.66,-143.5748],[6097.2975,-143.5748],[6112.7775,-142.512],[6124.0575,-142.512],[6132.3375,-143.1024],[6140.5,-143.1024],[6144.4575,-143.811],[6204.8775,-147.1175],[6220.4775,-147.2356],[6251.4975,-148.1803],[6270.76,-149.7155],[6286.24,-149.5974],[6293.7975,-148.7708],[6313.0575,-148.1803],[6401.7375,-148.1803],[6420.0375,-147.4718],[6475.54,-146.6451],[6515.8575,-146.9994],[6526.9575,-147.4718],[6560.7375,-147.9005],[6615.8175,-147.5894],[6619.42,-147.7449],[6627.2175,-147.4339],[6642.3975,-148.6782],[6649.8375,-148.6782],[6660.94,-148.9893],[6671.92,-144.9452],[6690.8175,-142.9232],[6756.76,-142.9232],[6765.76,-142.7209],[6769.8975,-142.3163],[6781.6,-142.3163],[6785.4375,-142.114],[6802.96,-142.114],[6820.5975,-142.5186],[6847.06,-142.5186],[6877.7775,-143.3278],[6913.7175,-143.7324],[6918.46,-144.3392],[6978.0975,-144.6518],[6982.7175,-145.1184],[6987.1575,-145.8961],[6991.84,-147.296],[6996.3375,-151.0289],[7001.0775,-154.2953],[7005.46,-156.1617],[7013.9175,-156.9394],[7018.4175,-158.8059],[7022.86,-159.4281],[7026.9375,-160.9835],[7031.02,-163.0055],[7035.8775,-164.872],[7040.3775,-166.1163],[7044.7575,-167.9828],[7049.1375,-172.3379],[7054.0575,-174.0488],[7071.2775,-173.7378],[7084.5975,-172.8045],[7194.7575,-172.649],[7199.1375,-173.2711],[7208.44,-173.2711],[7225.72,-174.3599],[7243.2375,-174.3599],[7257.1575,-173.6368],[7309.0575,-173.873],[7313.3175,-174.3453],[7330.4175,-174.8177],[7413.9375,-175.1288],[7429.2375,-174.1955],[7537.06,-173.8844],[7540.7775,-173.5734],[7697.2575,-173.7289],[7714.18,-172.7957],[7738.3,-172.7957],[7750.1775,-173.5734],[7765.36,-173.8844],[7769.2,-174.1955],[7781.1375,-174.1955],[7785.0975,-174.5066],[7864.2975,-174.5066],[7868.6175,-174.8177],[8177.5575,-175.2843],[8374.48,-175.2843],[8385.2775,-174.6621],[8469.5175,-174.9732],[8480.1975,-175.7509],[8930.9775,-176.2175],[8934.5775,-176.5286],[8942.0175,-177.7729],[8960.7375,-177.4619],[8964.64,-177.1508],[9090.94,-177.6174],[9381.9375,-177.6247],[9411.16,-176.7981]]
       }
       Scale.X {type=MovieSceneScriptingDoubleChannel, default=1, const [1,2]}
       Scale.Y {type=MovieSceneScriptingDoubleChannel, default=1, const [1,2]}
       Scale.Z {type=MovieSceneScriptingDoubleChannel, default=1, const [1,2]}
      ]
     }]
   }
  ]
 }
]
display_rate {denominator=1, numerator=60}
tick_resolution {denominator=1, numerator=24000}
tracks [{type=MovieSceneCameraCutTrack, sections [{type=MovieSceneCameraCutSection, end_frame=50546, start_frame=3453}]}]
